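# Physics asset for the Ice Claw two-handed melee pickaxe mesh, adding physical motion to two leaf-joint bones on the weapon.
# Capsule bodies cover the handle bone and the two dyn_leaf joints, with the joint capsules offset and rotated from their bones.
# Each leaf joint is constrained directly to the handle with tight swing and twist limits, keeping any movement to a slight wobble.
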PhysicsAsset /BRCosmetics/Weapons/FORT_Melee/Pickaxe_Ice_Claw/Meshes/Ice_Claw_Physics
bodies {
 Melee_TwoHanded_Handle [{shapeType=Capsule, length=12, radius=3}]
 dyn_leaf_jnt_1 [{shapeType=Capsule, length=5, radius=1.8, center=<4.5905 1.4349 -1.5191>, rotation=<rot -71.2917 176.2113 -159.8785>}]
 dyn_leaf_jnt_2 [{shapeType=Capsule, length=5, radius=1.8, center=<4.9848 0.4398 -0.6306>, rotation=<rot -82.5397 169.2704 -164.3552>}]
}
constraints [
 {bone1Name=dyn_leaf_jnt_1, bone2Name=Melee_TwoHanded_Handle, swing1LimitDegrees=10, swing2LimitDegrees=5, twistLimitDegrees=5}
 {bone1Name=dyn_leaf_jnt_2, bone2Name=Melee_TwoHanded_Handle, swing1LimitDegrees=4, swing2LimitDegrees=5, twistLimitDegrees=5}
]
